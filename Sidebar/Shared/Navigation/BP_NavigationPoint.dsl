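; BP_NavigationPoint is a sidebar navigation point that holds a social user, conversation ID and scroll offset.
; IsEqualTo casts the other point to this class, then requires the parent check plus matching social user and conversation ID, returning false on a failed cast.
; Scroll offset is not compared in this override.
; GetDebugInfo constructs a temporary social user name view model to resolve the user's name, then appends the conversation ID and scroll offset.
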(fn GetDebugInfo
  ;[1 1 1] Social User
  (bind _returnvalue (Game|ConstructObjectfromClass "/Script/FortniteUI.FortSocialUserNameViewModel"))
  (Class|FortSocialUserViewmodel|SetSocialUser _returnvalue (Variables|Data|GetSocialUser))
  (return (Utilities|String|Append "SocialUser: " (Utilities|Struct|BreakFortSocialUserNamePackage (Variables|/Script/FortniteUI.FortSocialUserNameViewModel|GetSocialUserNames _returnvalue)) " ConversationId: " (Utilities|String|ToString(Name) (Variables|Data|GetConversationId)) " ScrollOffset: " (Utilities|String|ToString(Float) (Variables|Data|GetScrollOffset)))))

(fn IsEqualTo
  (params (Other :type Fort%20Sidebar%20Navigation%20Point%20Object%20Reference))
  (bind _asbp_navigation_point (Utilities|Casting|CastToBP_NavigationPoint Other)
    (:then
      (bind _returnvalue (|CallParentFunction _asbp_navigation_point))
      (return (Math|Boolean|ANDBoolean _returnvalue (== (Variables|Data|GetSocialUser _asbp_navigation_point) (Variables|Data|GetSocialUser)) (== (Variables|Data|GetConversationId _asbp_navigation_point) (Variables|Data|GetConversationId)))))
    (:CastFailed
      (return false))))
